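# Physics asset for the Rocker Punk face accessory mesh, which gives its mohawk bones limited secondary motion relative to the head.
# Five dynamic mohawk bones each carry a capsule body, and the head carries a single sphere body.
# Each mohawk bone is constrained directly to the head rather than chained to its neighbor, with narrow swing limits and almost no twist.
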PhysicsAsset /BRCosmetics/Accessories/Hats/M_MED_RockerPunk_FaceAccessory/Meshes/M_MED_Rocker_Punk_FaceAcc_Physics
bodies {
 dyn_mohawk_01 [{shapeType=Capsule, length=4.9616, radius=2.7863, center=<1.5195 -0.026 0.0019>, rotation=<rot 0.562 77.2378 -91.4386>}]
 dyn_mohawk_02 [{shapeType=Capsule, length=3.8109, radius=1.9048, center=<0.3174 -0.1094 0.0005>, rotation=<rot -0.0908 85.9582 -89.599>}]
 dyn_mohawk_03 [{shapeType=Capsule, length=5.8432, radius=1.7317, center=<0.9563 -0.1015 -0.0002>, rotation=<rot 0 -92.6371 -89.8661>}]
 dyn_mohawk_04 [{shapeType=Capsule, length=5.0621, radius=1.7293, center=<0.5769 -0.1921 -0.0105>, rotation=<rot 0 -92.433 -88.3579>}]
 dyn_mohawk_05 [{shapeType=Capsule, length=3.1142, radius=1.9769, center=<0.8399 0.085 -0.0024>, rotation=<rot 0 -88.4171 -89.6534>}]
 head [{shapeType=Sphere, radius=9.7914, center=<3.4342 3.4999 0.0002>}]
}
constraints [
 {bone1Name=dyn_mohawk_01, bone2Name=head, swing1LimitDegrees=15, swing2LimitDegrees=15, twistLimitDegrees=1}
 {bone1Name=dyn_mohawk_02, bone2Name=head, swing1LimitDegrees=15, swing2LimitDegrees=15, twistLimitDegrees=1}
 {bone1Name=dyn_mohawk_03, bone2Name=head, swing1LimitDegrees=15, swing2LimitDegrees=15, twistLimitDegrees=1}
 {bone1Name=dyn_mohawk_04, bone2Name=head, swing1LimitDegrees=15, swing2LimitDegrees=15, twistLimitDegrees=1}
 {bone1Name=dyn_mohawk_05, bone2Name=head, swing1LimitDegrees=15, swing2LimitDegrees=15, twistLimitDegrees=1}
]
